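# Translucent, unlit, two-sided Niagara mesh particle material for the blade's emissive FX, named for other players' distant view.
# Particle color feeds the emissive adjustment function.
# The PanRed dynamic parameter drives panning of a noise-distorted gradient, sharpened by a power into a bright line and masked by IB_Mask.
# That mask adds to particle alpha, takes a distance-synced fade except on mobile, then gets UV-gradient and Fresnel shaping.
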Material /BRCosmetics/Weapons/FORT_Melee/Effects/Blade/M_IBlade_EmissiveFX_OtherPlayers_Distant
graph {
 properties {
  bUsedWithNiagaraMeshParticles=true
  blendMode=BLEND_Translucent
  pixelDepthOffsetMode=PDOM_Legacy
  shadingModel=MSM_Unlit
  twoSided=true
  usedShadingModels=Unlit
 }
 nodes [
  n1 = ParticleColor()
  MaterialFunctionCall(Emissive=@n1.RGB) {materialFunction=&MF_FX_EmissiveAdjustments}
  n2 = Multiply(A=Add(A=DynamicParameter() {defaultValue=<rgba 0.5 0.5 0.5 1>, paramNames [PanRed,PanGreen,PanBlue,Param4]}.PanRed) {constB=-3.2}) {constB=5}
  n3 = TextureCoordinate()
  n4 = ScalarParameter() {defaultValue=2, parameterName=Distort_Texture_Tiling}
  n5 = TextureSample(UVs=Panner(Coordinate=Multiply(A=@n3, B=@n4), Time=@n2) {speedY=1}) {texture=&T_NoisePack_01}
  n6 = Add(
   A=Panner(Coordinate=TextureCoordinate(), Time=@n2) {speedY=0.25}
   B=Multiply(A=ComponentMask(None=@n5.RGB) {g=true, r=true}, B=ScalarParameter() {defaultValue=0.1, parameterName=Dist_Mult})
  )
  n7 = Constant2Vector() {g=3}
  n8 = TextureSampleParameter2D(UVs=Multiply(A=Add(A=@n6, B=@n7)) {constB=-1}) {parameterName=TexRed, texture=/Game/Global/Textures/Default/Masks_and_Gradients/T_Vert_Gradient_Soft}
  n9 = ScalarParameter() {defaultValue=15, parameterName=Line_Power}
  n10 = Add(
   A=Multiply(A=Power(Base=@n8.RGB, Exp=@n9)) {constB=20}
   B=Multiply(A=@n8.RGB, B=ScalarParameter() {defaultValue=0.1, parameterName=Background_Texture_Intensity})
  )
  n11 = Multiply(
   A=TextureSample() {samplerType=SAMPLERTYPE_Masks, texture=/Game/Weapons/FORT_Melee/Blade_Athena/Textures/IB_Mask}.R
   B=Clamp(None=Multiply(B=@n10) {constA=4})
  )
  Multiply(A=@n1.A, B=@n11) {constA=1}
  n12 = Add(A=@n1.A, B=@n11)
  n13 = ScalarParameter() {defaultValue=1000, parameterName="Fade_Start_Distance (S)"}
  n14 = ScalarParameter() {defaultValue=500, parameterName="Fade_End_Distance (S)"}
  n15 = ComponentMask(None=TextureCoordinate()) {g=true}
  n16 = Multiply(
   A=ShadingPathSwitch(
    Default=MaterialFunctionCall(Opacity In=@n12, Fade_Start_Distance=@n13, Fade_End_Distance=@n14) {materialFunction=&MF_Distance_SyncedFade_FX}.Result
    Mobile=@n12
   ) {desc=Migrated from FeatureLevelSwitch}
   B=Multiply(A=@n15, B=@n15)
  )
  n17 = ScalarParameter() {defaultValue=5, parameterName="Exponent (S)"}
  n18 = Fresnel(
   ExponentIn=ScalarParameter() {defaultValue=1, parameterName=ExponentIn}
   BaseReflectFractionIn=ScalarParameter() {defaultValue=0.1, parameterName=BaseReflectFractionIn}
  )
  Multiply(
   A=@n16
   B=Multiply(
    A=MaterialFunctionCall(Exponent=@n17) {materialFunction=&AxisAlignedFresnel}.Fresnel
    B=OneMinus(None=@n18)
   )
  )
  Multiply()
  n19 = ComponentMask(None=TextureCoordinate()) {g=true}
  n20 = Multiply(A=@n19, B=@n19)
  Multiply(A=@n16, B=Power(Base=OneMinus(None=@n20), Exp=ScalarParameter() {defaultValue=5, parameterName=BottomStrength}))
 ]
}
